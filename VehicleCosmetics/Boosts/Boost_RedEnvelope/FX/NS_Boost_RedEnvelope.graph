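# Niagara system for the Red Envelope vehicle boost cosmetic, trailing envelope, money and red spark sprites from the boost ports.
# Every emitter partitions particles across filtered skeletal-mesh sockets and launches them along negative X.
# Three envelope layers spin under curved local acceleration; two add curl noise and randomized scale curves dipping negative for flutter.
# Money animates via SubUV, sparks swirl with vortex and noise, and the painted-variant override is disabled throughout.
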NiagaraSystem /VehicleCosmetics/Boosts/Boost_RedEnvelope/FX/NS_Boost_RedEnvelope
dependencies {
 usedDataInterfaces [
  &NiagaraDataInterfaceCurve
  &NiagaraDataInterfaceVector2DCurve
  &NiagaraDataInterfaceVectorCurve
  &NiagaraDataInterfaceSpriteRendererInfo
  &NiagaraDataInterfaceSkeletalMesh
 ]
 usedDynamicInputs [&MakeFloatFromVector2d]
 usedModules [
  &self:SystemSpawnScript.SystemScriptSource.SystemScriptGraph.NiagaraNodeAssignment_0.SetVariables_#1
  &SystemState
 ]
}
emitters {
 envelope0 {
  emitterUpdateScript [
   &EmitterState()
   &VC_FilteredSocketCount()
   &SpawnRate(SpawnRate=10, Use Spawn Probability=false)
   &SpawnPerUnit(Module.Use Max Movement Threshold=true, Module.Use Movement Tolerance=false, Use Spawn Probability=false) {enabled=false}
  ]
  particleSpawnScript [
   &VC_PaintedVariantsOverride() {enabled=false}
   &InitializeParticle(
    Color Mode=::NewEnumerator0
    Lifetime=&RandomRangeFloat(Minimum=1.5, Maximum=1)
    Sprite Rotation Mode=::NewEnumerator2
    Sprite Size=&Multiply_Vector2D_ByFloat(Float=2, Vector2D=<15 30>)
    Sprite Size Mode=::NewEnumerator4
    UsePositionOffset=false
    Write Material Random=true
   )
   &PartitionParticles(Number Of Partitions=$Emitter.PortCount)
   &SkeletalMeshLocation(
    Bone / Socket Index=$Particles.PartitionParticles.Partition
    Bone / Socket Sampling Mode=::NewEnumerator5
    MaskSpawnGroup=false
    Mesh Sampling Type=::NewEnumerator6
    Skeletal Mesh=$System.SkeletalMesh
   )
   &AddVelocity(Velocity=&RandomRangeVector(Minimum=<1000 100 300>, Maximum=<800 -100 0>), Velocity Speed Scale=-1)
   &ShapeLocation(MaskSpawnGroup=false, Override Seed=false, Sphere Radius=20, UseOwnerScale=true)
  ]
  particleUpdateScript [
   &ParticleState()
   &SpriteRotationRate(Rotation Rate=&FloatFromCurve(
     FloatCurve=DataInterface() {propertyValues {Curve {keys [<at 0.2 -0.05 RCIM_Cubic>, {interpMode=RCIM_Cubic, tangentMode=RCTM_User, time=0.3, value=1}, <at 1 1>]}}}
     Scale Curve=&RandomRangeFloat(Minimum=-300, Maximum=300)
    )) {module=SpriteRotationRate001}
   &ScaleSpriteSize(Uniform Curve Sprite Scale=DataInterface() {propertyValues {Curve {keys [{tangentMode=RCTM_User, time=0, value=1}, {interpMode=RCIM_Cubic, tangentMode=RCTM_User, time=1, value=1}, <at 1.5 0 RCIM_Cubic>]}}})
   &AccelerationForce(
    Acceleration=&VectorFromCurve(
     VectorCurve=DataInterface() {propertyValues {
       XCurve {keys [<at 0 1>, {arriveTangent=-1, interpMode=RCIM_Cubic, leaveTangent=-1, time=0.5, value=1}, <at 1 0 RCIM_Cubic>]}
       YCurve {keys [<at 0 1>,<at 1 0>]}
       ZCurve {keys [<at 0 0.5>,<at 1 0>]}
      }}
     Scale Curve=<800 1 -500>
    )
    Coordinate Space=::Local
   )
   &SolveForcesAndVelocity(Clamp Velocity=false, Limit Acceleration=false)
   &DynamicMaterialParameters(Module.Index0Param0Write=true, Module.Index0Param1Write=false, Module.Index0Param2Write=false, Module.Index0Param3Write=false)
  ]
  rendererClasses [&NiagaraSpriteRendererProperties]
  renderers [{rendererClass=&NiagaraSpriteRendererProperties}]
 }
 envelope0004 {
  emitterUpdateScript [
   &EmitterState()
   &VC_FilteredSocketCount()
   &SpawnRate(SpawnRate=20, Use Spawn Probability=false)
   &SpawnPerUnit(Module.Use Max Movement Threshold=true, Module.Use Movement Tolerance=false, Use Spawn Probability=false) {enabled=false}
  ]
  particleSpawnScript [
   &VC_PaintedVariantsOverride() {enabled=false}
   &InitializeParticle(
    Color Mode=::NewEnumerator0
    Lifetime=&RandomRangeFloat(Minimum=1.5, Maximum=1)
    Sprite Rotation Mode=::NewEnumerator2
    Sprite Size=&Multiply_Vector2D_ByFloat(Float=2, Vector2D=<15 30>)
    Sprite Size Mode=::NewEnumerator4
    UsePositionOffset=false
    Write Material Random=true
   )
   &PartitionParticles(Number Of Partitions=$Emitter.PortCount)
   &SkeletalMeshLocation(
    Bone / Socket Index=$Particles.PartitionParticles.Partition
    Bone / Socket Sampling Mode=::NewEnumerator5
    MaskSpawnGroup=false
    Mesh Sampling Type=::NewEnumerator6
    Skeletal Mesh=$System.SkeletalMesh
   )
   &AddVelocity(Velocity=&RandomRangeVector(Minimum=<1000 100 -200>, Maximum=<1000 -100 100>), Velocity Speed Scale=-1)
   &ShapeLocation(MaskSpawnGroup=false, Override Seed=false, Sphere Radius=20, UseOwnerScale=true)
  ]
  particleUpdateScript [
   &ParticleState()
   &SpriteRotationRate(Rotation Rate=&FloatFromCurve(
     FloatCurve=DataInterface() {propertyValues {Curve {keys [<at 0.2 -0.05 RCIM_Cubic>, {interpMode=RCIM_Cubic, tangentMode=RCTM_User, time=0.3, value=1}, <at 1 1>]}}}
     Scale Curve=&RandomRangeFloat(Minimum=-300, Maximum=300)
    )) {module=SpriteRotationRate001}
   &ScaleSpriteSize(
    Scale Factor=&RandomRangeVector2D(
     Minimum=&Vector2DFromCurve(
      Vector2Curve=DataInterface() {propertyValues {
        XCurve {keys [
          <at 0 1 RCIM_Cubic>
          <at 0.3 0.3 RCIM_Cubic>
          {interpMode=RCIM_Cubic, tangentMode=RCTM_User, time=0.5, value=1}
          {arriveTangent=-2, interpMode=RCIM_Cubic, leaveTangent=-2, time=0.8, value=0.6}
          <at 1 0 RCIM_Cubic>
         ]}
        YCurve {keys [
          <at 0 1 RCIM_Cubic>
          {arriveTangent=-4, interpMode=RCIM_Cubic, leaveTangent=-4, time=0.2, value=0.8}
          {arriveTangent=0.3333, interpMode=RCIM_Cubic, leaveTangent=0.3333, time=0.5, value=-1}
          {arriveTangent=2, interpMode=RCIM_Cubic, leaveTangent=2, time=0.8, value=1}
          <at 1 0 RCIM_Cubic>
         ]}
       }}
      ScaleCurve=<1 1>
     )
     Maximum=&Vector2DFromCurve(
      Vector2Curve=DataInterface() {propertyValues {
        XCurve {keys [
          <at 0 1 RCIM_Cubic>
          {arriveTangent=-0.6, interpMode=RCIM_Cubic, leaveTangent=-0.6, time=0.3, value=0.7}
          {arriveTangent=-0.5, interpMode=RCIM_Cubic, leaveTangent=-0.5, time=0.5, value=0.7}
          {arriveTangent=-1.4, interpMode=RCIM_Cubic, leaveTangent=-1.4, time=0.8, value=0.45}
          <at 1 0 RCIM_Cubic>
         ]}
        YCurve {keys [
          <at 0 1 RCIM_Cubic>
          {arriveTangent=-0.6, interpMode=RCIM_Cubic, leaveTangent=-0.6, time=0.2, value=0.375}
          {arriveTangent=0.2917, interpMode=RCIM_Cubic, leaveTangent=0.2917, time=0.5, value=0.7}
          {arriveTangent=-1.4, interpMode=RCIM_Cubic, leaveTangent=-1.4, time=0.8, value=0.55}
          <at 1 0 RCIM_Cubic>
         ]}
       }}
      ScaleCurve=<1 1>
     )
    )
    Scale Sprite Size Mode=::NewEnumerator2
   )
   &CurlNoiseForce(Noise Frequency=10, Noise Strength=500, Pan Noise=true)
   &AccelerationForce(
    Acceleration=&VectorFromCurve(
     VectorCurve=DataInterface() {propertyValues {
       XCurve {keys [<at 0 1>, {arriveTangent=-1, interpMode=RCIM_Cubic, leaveTangent=-1, time=0.5, value=1}, <at 1 0 RCIM_Cubic>]}
       YCurve {keys [<at 0 1>,<at 1 0>]}
       ZCurve {keys [<at 0 0.5>,<at 1 0>]}
      }}
     Scale Curve=<800 1 -500>
    )
    Coordinate Space=::Local
   )
   &SolveForcesAndVelocity(Clamp Velocity=false, Limit Acceleration=false)
   &DynamicMaterialParameters(Module.Index0Param0Write=true, Module.Index0Param1Write=false, Module.Index0Param2Write=false, Module.Index0Param3Write=false)
  ]
  rendererClasses [&NiagaraSpriteRendererProperties]
  renderers [{rendererClass=&NiagaraSpriteRendererProperties}]
 }
 envelope0005 {
  emitterUpdateScript [
   &EmitterState()
   &VC_FilteredSocketCount()
   &SpawnRate(SpawnRate=30, Use Spawn Probability=false)
   &SpawnPerUnit(Module.Use Max Movement Threshold=true, Module.Use Movement Tolerance=false, Use Spawn Probability=false) {enabled=false}
  ]
  particleSpawnScript [
   &VC_PaintedVariantsOverride() {enabled=false}
   &InitializeParticle(
    Color Mode=::NewEnumerator0
    Lifetime=&RandomRangeFloat(Minimum=1.5, Maximum=1)
    Sprite Size=&Multiply_Vector2D_ByFloat(Float=2, Vector2D=<15 30>)
    Sprite Size Mode=::NewEnumerator4
    UsePositionOffset=false
    Write Material Random=true
   )
   &PartitionParticles(Number Of Partitions=$Emitter.PortCount)
   &SkeletalMeshLocation(
    Bone / Socket Index=$Particles.PartitionParticles.Partition
    Bone / Socket Sampling Mode=::NewEnumerator5
    MaskSpawnGroup=false
    Mesh Sampling Type=::NewEnumerator6
    Skeletal Mesh=$System.SkeletalMesh
   )
   &AddVelocity(Velocity=&RandomRangeVector(Minimum=<1000 100 -200>, Maximum=<1000 -100 100>), Velocity Speed Scale=-1)
   &ShapeLocation(MaskSpawnGroup=false, Override Seed=false, Sphere Radius=30, UseOwnerScale=true)
  ]
  particleUpdateScript [
   &ParticleState()
   &SpriteRotationRate(Rotation Rate=&FloatFromCurve(
     FloatCurve=DataInterface() {propertyValues {Curve {keys [<at 0.2 -0.05 RCIM_Cubic>, {interpMode=RCIM_Cubic, tangentMode=RCTM_User, time=0.3, value=1}, <at 1 1>]}}}
     Scale Curve=&RandomRangeFloat(Minimum=-300, Maximum=300)
    )) {module=SpriteRotationRate001}
   &ScaleSpriteSize(
    Scale Factor=&RandomRangeVector2D(
     Minimum=&Vector2DFromCurve(
      Vector2Curve=DataInterface() {propertyValues {
        XCurve {keys [
          <at 0 1 RCIM_Cubic>
          <at 0.3 0.3 RCIM_Cubic>
          {interpMode=RCIM_Cubic, tangentMode=RCTM_User, time=0.5, value=1}
          {arriveTangent=-2, interpMode=RCIM_Cubic, leaveTangent=-2, time=0.8, value=0.6}
          <at 1 0 RCIM_Cubic>
         ]}
        YCurve {keys [
          <at 0 1 RCIM_Cubic>
          {arriveTangent=-4, interpMode=RCIM_Cubic, leaveTangent=-4, time=0.2, value=0.8}
          {arriveTangent=0.3333, interpMode=RCIM_Cubic, leaveTangent=0.3333, time=0.5, value=-1}
          {arriveTangent=2, interpMode=RCIM_Cubic, leaveTangent=2, time=0.8, value=1}
          <at 1 0 RCIM_Cubic>
         ]}
       }}
      ScaleCurve=<1 1>
     )
     Maximum=&Vector2DFromCurve(
      Vector2Curve=DataInterface() {propertyValues {
        XCurve {keys [
          <at 0 1 RCIM_Cubic>
          {arriveTangent=-0.6, interpMode=RCIM_Cubic, leaveTangent=-0.6, time=0.3, value=0.7}
          {arriveTangent=-0.5, interpMode=RCIM_Cubic, leaveTangent=-0.5, time=0.5, value=0.7}
          {arriveTangent=-1.4, interpMode=RCIM_Cubic, leaveTangent=-1.4, time=0.8, value=0.45}
          <at 1 0 RCIM_Cubic>
         ]}
        YCurve {keys [
          <at 0 1 RCIM_Cubic>
          {arriveTangent=-0.6, interpMode=RCIM_Cubic, leaveTangent=-0.6, time=0.2, value=0.375}
          {arriveTangent=0.2917, interpMode=RCIM_Cubic, leaveTangent=0.2917, time=0.5, value=0.7}
          {arriveTangent=-1.4, interpMode=RCIM_Cubic, leaveTangent=-1.4, time=0.8, value=0.55}
          <at 1 0 RCIM_Cubic>
         ]}
       }}
      ScaleCurve=<1 1>
     )
    )
    Scale Sprite Size Mode=::NewEnumerator2
   )
   &CurlNoiseForce(Noise Frequency=10, Noise Strength=800, Pan Noise=true)
   &AccelerationForce(
    Acceleration=&VectorFromCurve(
     VectorCurve=DataInterface() {propertyValues {
       XCurve {keys [<at 0 1>, {arriveTangent=-1, interpMode=RCIM_Cubic, leaveTangent=-1, time=0.5, value=1}, <at 1 0 RCIM_Cubic>]}
       YCurve {keys [<at 0 1>,<at 1 0>]}
       ZCurve {keys [<at 0 0.5>,<at 1 0>]}
      }}
     Scale Curve=<800 1 -500>
    )
    Coordinate Space=::Local
   )
   &SolveForcesAndVelocity(Clamp Velocity=false, Limit Acceleration=false)
   &DynamicMaterialParameters(
    Index 0 Param 1=1
    Module.Index0Param0Write=true
    Module.Index0Param1Write=false
    Module.Index0Param2Write=false
    Module.Index0Param3Write=false
   )
  ]
  rendererClasses [&NiagaraSpriteRendererProperties]
  renderers [{rendererClass=&NiagaraSpriteRendererProperties}]
 }
 money {
  emitterUpdateScript [
   &EmitterState()
   &VC_FilteredSocketCount()
   &SpawnRate(SpawnRate=15, Use Spawn Probability=false)
  ]
  particleSpawnScript [
   &VC_PaintedVariantsOverride() {enabled=false}
   &InitializeParticle(
    Color Mode=::NewEnumerator0
    Lifetime=&RandomRangeFloat(Minimum=1.5, Maximum=1)
    Sprite Rotation Mode=::NewEnumerator2
    Sprite Size Mode=::NewEnumerator3
    Uniform Sprite Size=100
    UsePositionOffset=false
    Write Material Random=true
   )
   &PartitionParticles(Number Of Partitions=$Emitter.PortCount)
   &SkeletalMeshLocation(
    Bone / Socket Index=$Particles.PartitionParticles.Partition
    Bone / Socket Sampling Mode=::NewEnumerator5
    MaskSpawnGroup=false
    Mesh Sampling Type=::NewEnumerator6
    Skeletal Mesh=$System.SkeletalMesh
   )
   &AddVelocity(Velocity=&RandomRangeVector(Minimum=<800 50 50>, Maximum=<600 -50 -200>), Velocity Speed Scale=-1)
   &ShapeLocation(MaskSpawnGroup=false, Override Seed=false, Sphere Radius=20, UseOwnerScale=true)
  ]
  particleUpdateScript [
   &ParticleState()
   &ScaleSpriteSize(Uniform Curve Sprite Scale=DataInterface() {propertyValues {Curve {keys [{tangentMode=RCTM_User, time=0, value=1}, {interpMode=RCIM_Cubic, tangentMode=RCTM_User, time=1, value=1}, <at 1.5 0 RCIM_Cubic>]}}})
   &AccelerationForce(
    Acceleration=&VectorFromCurve(
     VectorCurve=DataInterface() {propertyValues {
       XCurve {keys [<at 0 1>, {arriveTangent=-1, interpMode=RCIM_Cubic, leaveTangent=-1, time=0.5, value=1}, <at 1 0 RCIM_Cubic>]}
       YCurve {keys [<at 0 0>,<at 1 0>]}
       ZCurve {keys [<at 0 0.5>,<at 1 0>]}
      }}
     Scale Curve=<600 1 -500>
    )
    Coordinate Space=::Local
   )
   &SubUVAnimation(
    "Sprite Renderer "=DataInterface() {propertyValues {SpriteRenderer=&self:money_0.NiagaraSpriteRendererProperties_0}}
    UseEndFrame=false
    UseStartFrame=false
   )
   &SolveForcesAndVelocity(Clamp Velocity=false, Limit Acceleration=false)
  ]
  rendererClasses [&NiagaraSpriteRendererProperties]
  renderers [{rendererClass=&NiagaraSpriteRendererProperties}]
 }
 sparks_001 {
  emitterUpdateScript [
   &EmitterState()
   &VC_FilteredSocketCount()
   &SpawnRate(SpawnRate=120, Use Spawn Probability=false)
  ]
  particleSpawnScript [
   &VC_PaintedVariantsOverride() {enabled=false}
   &InitializeParticle(
    Color Mode=::NewEnumerator0
    Lifetime=&RandomRangeFloat(Minimum=0.3, Maximum=0.6)
    Sprite Rotation Mode=::NewEnumerator2
    Sprite Size Max=<0.8 4>
    Sprite Size Min=<1 5>
    Sprite Size Mode=::NewEnumerator2
    UsePositionOffset=false
    Write Material Random=true
   )
   &PartitionParticles(Number Of Partitions=$Emitter.PortCount)
   &SkeletalMeshLocation(
    Bone / Socket Index=$Particles.PartitionParticles.Partition
    Bone / Socket Sampling Mode=::NewEnumerator5
    MaskSpawnGroup=false
    Mesh Sampling Type=::NewEnumerator6
    Skeletal Mesh=$System.SkeletalMesh
   )
   &AddVelocity(
    Velocity=&RandomRangeVector(Minimum=<1000 300 300>, Maximum=<1000 -300 -300>)
    Velocity Speed Scale=-0.3
   )
   &ShapeLocation(MaskSpawnGroup=false, Override Seed=false, Sphere Radius=30, UseOwnerScale=true)
   &Color(Color=<rgba 10 0 0 1>)
  ]
  particleUpdateScript [
   &ParticleState()
   &ScaleSpriteSize(
    Non-Uniform Curve Sprite Scale=DataInterface() {propertyValues {
      XCurve {keys [<at 0 0 RCIM_Cubic>, {arriveTangent=1, interpMode=RCIM_Cubic, leaveTangent=1, time=0.3, value=1}, <at 1 1 RCIM_Cubic>]}
      YCurve {keys [<at 0 0 RCIM_Cubic>, {arriveTangent=1, interpMode=RCIM_Cubic, leaveTangent=1, time=0.3, value=2}, <at 1 1 RCIM_Cubic>]}
     }}
    Scale Sprite Size Mode=::NewEnumerator3
   )
   &VortexForce(
    UseOriginOffset=false
    Vortex Axis=&RandomRangeVector(Minimum=<-1 -1 -1>, Maximum=<1 1 1>)
    Vortex Force Amount=500
   )
   &CurlNoiseForce(Noise Strength=1000, Pan Noise=false) {module=CurlNoiseForce001}
   &SolveForcesAndVelocity(Clamp Velocity=false, Limit Acceleration=false)
   &ScaleColor(
    Scale Alpha=&FloatFromCurve(FloatCurve=DataInterface() {propertyValues {Curve {keys [
         {interpMode=RCIM_Cubic, tangentMode=RCTM_User, time=0, value=0}
         {interpMode=RCIM_Cubic, tangentMode=RCTM_User, time=0.1, value=1}
         {interpMode=RCIM_Cubic, tangentMode=RCTM_User, time=0.9, value=1}
         <at 1 0 RCIM_Cubic>
        ]}}})
    Scale RGB=&Multiply_VectorByFloat(Vector=<1 1 1>, Float=30)
    ScaleA=true
    ScaleRGB=true
   )
   &DynamicMaterialParameters(
    Index 0 Param 1=50
    Module.Index0Param0Write=true
    Module.Index0Param1Write=false
    Module.Index0Param2Write=false
    Module.Index0Param3Write=false
   )
  ]
  rendererClasses [&NiagaraSpriteRendererProperties]
  renderers [{rendererClass=&NiagaraSpriteRendererProperties}]
 }
}
summary {
 systemName=NS_Boost_RedEnvelope
 emitters [
  {emitterName=sparks_001, rendererClasses [&NiagaraSpriteRendererProperties]}
  {emitterName=money, rendererClasses [&NiagaraSpriteRendererProperties]}
  {emitterName=envelope0, rendererClasses [&NiagaraSpriteRendererProperties]}
  {emitterName=envelope0004, rendererClasses [&NiagaraSpriteRendererProperties]}
  {emitterName=envelope0005, rendererClasses [&NiagaraSpriteRendererProperties]}
 ]
 userVariables [
  User.bIsFrontendPreview {type=NiagaraBool}
  User.bIsPainted {type=NiagaraBool}
  User.ColorVariantID {type=NiagaraInt32}
  User.PaintedColor {type=LinearColor, description="", defaultValue {struct=LinearColor, value=<rgba 1 1 1 1>}}
  User.PaintedScalar {type=NiagaraFloat, defaultValue=1, description=""}
  User.PortCount {type=NiagaraInt32, defaultValue=1}
 ]
}
