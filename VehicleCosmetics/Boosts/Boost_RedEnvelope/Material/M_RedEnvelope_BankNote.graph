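# Unlit, masked, two-sided Niagara particle material for the Red Envelope boost cosmetic, drawing a red envelope with an optional banknote strip.
# A dirt-modulated red base gets an orange emblem from rotated rim textures clipped by a procedural circle, feeding an emissive adjustment function.
# Dynamic parameter R, via Ceil, blends in a suitcase-texture band near the top.
# The same switch extends the square-texture cutout mask to include that band.
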Material /VehicleCosmetics/Boosts/Boost_RedEnvelope/Material/M_RedEnvelope_BankNote
graph {
 properties {
  bUsedWithNiagaraMeshParticles=true
  bUsedWithNiagaraSprites=true
  blendMode=BLEND_Masked
  pixelDepthOffsetMode=PDOM_Legacy
  refractionMethod=RM_2DOffset
  shadingModel=MSM_Unlit
  twoSided=true
  usedShadingModels=Unlit
 }
 nodes [
  Constant2Vector() {g=-0.1}
  Constant()
  n1 = DynamicParameter()
  Floor(None=@n1.R)
  n2 = TextureCoordinate() {uTiling=0.75}
  n3 = Constant2Vector() {g=0.1, r=-0.425}
  n4 = Constant() {r=1.5}
  n5 = Multiply(A=Rotator(Coordinate=Add(A=@n2, B=@n3), Time=Constant() {r=0.85}) {centerX=0.75, centerY=1, speed=3.14}, B=@n4)
  n6 = ComponentMask(None=TextureSample(UVs=@n5) {texture=/VehicleCosmetics/Boosts/Boost_RedEnvelope/Texture/T_RedEnvelope_rim_cnny_d}.G) {r=true}
  MaterialExpressionConstantClamp_3 = Saturate(None=ConstantBiasScale(None=OneMinus(None=@n6)) {bias=-0.8, scale=5})
  n7 = Multiply(A=@n4, B=Add(A=TextureCoordinate() {uTiling=2, vTiling=2.6}, B=Constant2Vector() {g=-1.55, r=-1}))
  n8 = DotProduct(A=@n7, B=@n7)
  MaterialExpressionConstantClamp_4 = Saturate(None=Multiply(A=OneMinus(None=@n8), B=Constant() {r=60}))
  n9 = Multiply(A=@MaterialExpressionConstantClamp_3, B=@MaterialExpressionConstantClamp_4)
  MaterialExpressionConstantClamp_2 = Saturate(None=OneMinus(None=@n9))
  n10 = Multiply(
   A=@MaterialExpressionConstantClamp_2
   B=TextureSample() {samplerType=SAMPLERTYPE_LinearColor, texture=/VehicleCosmetics/Boosts/Boost_RedEnvelope/Texture/T_RedEnvelope_Noise_Dirt_01_N}.RGB
  )
  MaterialExpressionConstantClamp_15 = Saturate(None=ConstantBiasScale(None=Desaturation(None=@n10)) {bias=0.8, scale=1})
  n11 = Multiply(A=@MaterialExpressionConstantClamp_15, B=Constant3Vector() {constant=<rgba 1 0 0 0>})
  n12 = Constant3Vector() {constant=<rgba 1 0.5 0 0>}
  MaterialExpressionConstantClamp_1 = Saturate(None=Multiply(
    A=TextureSample(UVs=@n5) {samplerType=SAMPLERTYPE_LinearColor, texture=/VehicleCosmetics/Boosts/Boost_RedEnvelope/Texture/T_RedEnvelope_rim_cnny_n}.RGB
    B=@MaterialExpressionConstantClamp_4
   ))
  n13 = Desaturation(None=@MaterialExpressionConstantClamp_1)
  n14 = Multiply(A=Constant() {r=1}, B=Add(A=LinearInterpolate(A=@n11, B=@n12, Alpha=@n9), B=Multiply(A=@n12, B=@n13)))
  n15 = Add(A=TextureCoordinate() {uTiling=0.3, vTiling=0.5}, B=Constant2Vector() {r=0.74})
  n16 = ComponentMask(None=TextureCoordinate()) {g=true}
  n17 = OneMinus(None=@n16)
  MaterialExpressionConstantClamp_7 = Saturate(None=ConstantBiasScale(None=@n17) {bias=-0.72, scale=100})
  n18 = Add(A=TextureCoordinate() {uTiling=2, vTiling=2}, B=Constant() {r=-1})
  n19 = Abs(None=ComponentMask(None=@n18) {r=true})
  n20 = ConstantBiasScale(None=Constant()) {bias=0.4}
  MaterialExpressionConstantClamp_10 = Saturate(None=OneMinus(None=Add(A=@n19, B=@n20)))
  n21 = ComponentMask(None=@n18) {g=true}
  n22 = AppendVector(A=@MaterialExpressionConstantClamp_10, B=OneMinus(None=Abs(None=@n21)))
  MaterialExpressionConstantClamp_9 = Saturate(None=Multiply(A=Abs(None=@n22), B=Constant() {r=14}))
  MaterialExpressionConstantClamp_8 = Saturate(None=ConstantBiasScale(None=ComponentMask(None=@MaterialExpressionConstantClamp_9) {r=true}) {bias=-0.7, scale=100})
  n23 = Multiply(A=@MaterialExpressionConstantClamp_7, B=@MaterialExpressionConstantClamp_8)
  MaterialExpressionConstantClamp_6 = Saturate(None=Multiply(A=TextureSample(UVs=@n15) {texture=/VehicleCosmetics/Boosts/Boost_RedEnvelope/Texture/T_RedEnvelope_Suitcase_D}.RGB, B=@n23))
  MaterialExpressionConstantClamp_12 = Saturate(None=ConstantBiasScale(None=@n17) {bias=-0.7, scale=20})
  MaterialExpressionConstantClamp_13 = Saturate(None=OneMinus(None=Multiply(A=@MaterialExpressionConstantClamp_12, B=@MaterialExpressionConstantClamp_8)))
  n24 = Multiply(A=@n14, B=@MaterialExpressionConstantClamp_13)
  MaterialExpressionConstantClamp_5 = Saturate(None=Ceil(None=@n1.R))
  MaterialFunctionCall(Emissive=LinearInterpolate(A=@n14, B=Add(A=@MaterialExpressionConstantClamp_6, B=@n24), Alpha=@MaterialExpressionConstantClamp_5)) {materialFunction=&MF_FX_EmissiveAdjustments}
  Reroute()
  MaterialExpressionConstantClamp_0 = Saturate(None=Add(A=@n10, B=@MaterialExpressionConstantClamp_1))
  Reroute(None=@MaterialExpressionConstantClamp_0)
  n25 = Add(A=TextureCoordinate() {coordinateIndex=1}) {constB=0}
  MaterialExpressionConstantClamp_16 = Saturate(None=ConstantBiasScale(None=TextureSample(UVs=@n25) {texture=/VehicleCosmetics/Boosts/Boost_RedEnvelope/Texture/T_RedEnvelope_SquareMask_RGB}.G) {bias=-0.3, scale=4})
  MaterialExpressionConstantClamp_14 = Saturate(None=ConstantBiasScale(None=@n16) {bias=-0.1, scale=100})
  n26 = Multiply(A=@MaterialExpressionConstantClamp_16, B=@MaterialExpressionConstantClamp_14)
  n27 = Saturate(None=Add(A=@n26, B=@n23))
  Saturate(None=Multiply(A=LinearInterpolate(A=@n26, B=@n27, Alpha=@MaterialExpressionConstantClamp_5)) {constB=20})
 ]
}
